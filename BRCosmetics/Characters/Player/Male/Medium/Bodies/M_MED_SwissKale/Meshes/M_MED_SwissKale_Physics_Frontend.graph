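# Physics asset for the SwissKale male medium player body, named as the Frontend variant, simulating its front cloth flaps and a hanging rope with hook.
# A three-box center cloth chain hangs from the pelvis, while left and right panels constrain to both their thigh and the center's top box.
# A limited-swing rope chain runs from spine_01 to a multi-capsule hook, with thigh capsules supplying leg collision volumes.
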PhysicsAsset /BRCosmetics/Characters/Player/Male/Medium/Bodies/M_MED_SwissKale/Meshes/M_MED_SwissKale_Physics_Frontend
bodies {
 dyn_front_C_cloth_1 [{shapeType=Box, extentX=2.7414, extentY=14, extentZ=14, shapeName=dyn_front_L_cloth_1_box, center=<7 0 1>, rotation=<rot 90 0 0>}]
 dyn_front_C_cloth_2 [{shapeType=Box, extentX=4.8436, extentY=13.5, extentZ=14, shapeName=dyn_front_L_cloth_1_box, center=<7 0 0>, rotation=<rot 85.0001 0 0>}]
 dyn_front_C_cloth_3 [{shapeType=Box, extentX=4.4799, extentY=13.5, extentZ=14, shapeName=dyn_front_L_cloth_1_box, center=<7 0 0>, rotation=<rot 90 0 0>}]
 dyn_front_L_cloth_1 [{shapeType=Box, extentX=2.6736, extentY=14.1885, extentZ=24.2447, shapeName=dyn_front_L_cloth_1_box, center=<13 0 2.5>, rotation=<rot 90 0 0>}]
 dyn_front_R_cloth_1 [{
   shapeType=Box
   extentX=2.6736
   extentY=14.1885
   extentZ=24.2447
   shapeName=dyn_front_L_cloth_1_box
   center=<-13 -0.7253 -2.5>
   rotation=<rot 90 0 0>
  }]
 dyn_ropeHook [
  {shapeType=Capsule, length=20, radius=0.5, shapeName=dyn_ropeHook_capsule, center=<11.5744 3.7019 -0.872>, rotation=<rot 0 -68.8027 -87.5089>}
  {shapeType=Capsule, length=6, radius=1.3, shapeName=dyn_ropeHook_capsule, center=<20.3541 5.5024 -6.1948>}
  {shapeType=Capsule, length=6, radius=1.3, shapeName=dyn_ropeHook_capsule, center=<20.3541 5.5024 4.1886>}
  {
   shapeType=Capsule
   length=8
   radius=1.3
   shapeName=dyn_ropeHook_capsule
   center=<17.4126 10.5577 -2.4899>
   rotation=<rot -0.0001 10.0001 90.0001>
  }
 ]
 dyn_rope_bottom_B_1 [{shapeType=Capsule, length=3, radius=2, shapeName=dyn_rope_bottom_B_1_capsule, center=<2 0 0>, rotation=<rot 90 0 0>}]
 dyn_rope_bottom_B_2 [{shapeType=Capsule, length=3, radius=2, shapeName=dyn_rope_bottom_B_2_capsule, center=<2 0 0>, rotation=<rot 90 0 0>}]
 dyn_rope_master [{shapeType=Capsule, length=2.3721, radius=1.5893, shapeName=dyn_rope_master_capsule, center=<0.4731 0 0>, rotation=<rot 90 0 0>}]
 pelvis [{shapeType=Capsule, length=1, radius=1}]
 spine_01 [{shapeType=Capsule, length=1, radius=1, shapeName=spine_01_capsule}]
 thigh_l [
  {shapeType=Capsule, length=30, radius=12, shapeName=thigh_r_capsule, center=<-19.6499 -1.2556 -2.1657>, rotation=<rot 90 179.4515 178.5951>}
  {shapeType=Capsule, length=15, radius=14, shapeName=thigh_l_capsule, center=<-11 -3 7>, rotation=<rot -90 0 0>}
 ]
 thigh_r [
  {shapeType=Capsule, length=30, radius=12, shapeName=thigh_r_capsule, center=<19.6499 1.2556 2.1657>, rotation=<rot 90 178.6918 177.8354>}
  {shapeType=Capsule, length=15, radius=14, shapeName=thigh_r_capsule, center=<11 3 -3>, rotation=<rot 90 0 0>}
 ]
}
constraints [
 {bone1Name=dyn_rope_bottom_B_1, bone2Name=dyn_rope_master, swing1LimitDegrees=45, swing2LimitDegrees=45, twistLimitDegrees=10}
 {bone1Name=dyn_rope_bottom_B_2, bone2Name=dyn_rope_bottom_B_1, swing1LimitDegrees=45, swing2LimitDegrees=45, twistLimitDegrees=10}
 {bone1Name=dyn_ropeHook, bone2Name=dyn_rope_bottom_B_2, swing1LimitDegrees=45, swing2LimitDegrees=45, twistLimitDegrees=10}
 {bone1Name=dyn_front_C_cloth_2, bone2Name=dyn_front_C_cloth_1, swing1LimitDegrees=20, swing2LimitDegrees=60, twistLimitDegrees=20}
 {bone1Name=dyn_front_C_cloth_3, bone2Name=dyn_front_C_cloth_2, swing1LimitDegrees=20, swing2LimitDegrees=60, twistLimitDegrees=20}
 {
  bone1Name=thigh_l
  bone2Name=pelvis
  swing1LimitDegrees=45
  swing1Motion=Free
  swing2LimitDegrees=45
  swing2Motion=Free
  twistLimitDegrees=45
  twistMotion=Free
 }
 {
  bone1Name=thigh_r
  bone2Name=pelvis
  swing1LimitDegrees=45
  swing1Motion=Free
  swing2LimitDegrees=45
  swing2Motion=Free
  twistLimitDegrees=45
  twistMotion=Free
 }
 {bone1Name=dyn_front_C_cloth_1, bone2Name=pelvis, swing1LimitDegrees=20, swing2LimitDegrees=60, twistLimitDegrees=20}
 {bone1Name=dyn_front_L_cloth_1, bone2Name=thigh_l, swing1LimitDegrees=10, swing2LimitDegrees=45, twistLimitDegrees=45}
 {bone1Name=dyn_front_R_cloth_1, bone2Name=thigh_r, swing1LimitDegrees=10, swing2LimitDegrees=45, twistLimitDegrees=45}
 {bone1Name=dyn_front_L_cloth_1, bone2Name=dyn_front_C_cloth_1, swing1LimitDegrees=30, swing2LimitDegrees=30, twistLimitDegrees=30}
 {bone1Name=dyn_front_R_cloth_1, bone2Name=dyn_front_C_cloth_1, swing1LimitDegrees=30, swing2LimitDegrees=30, twistLimitDegrees=30}
 {bone1Name=dyn_rope_master, bone2Name=spine_01, swing1LimitDegrees=45, swing2LimitDegrees=45, twistLimitDegrees=10}
 {
  bone1Name=spine_01
  bone2Name=pelvis
  swing1LimitDegrees=45
  swing1Motion=Free
  swing2LimitDegrees=45
  swing2Motion=Free
  twistLimitDegrees=45
  twistMotion=Free
 }
]
