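# Animation Blueprint for the SwissKale face accessory cosmetic, animating its hood and drapes.
# AnimGraph blends parent and body upper-arm poses, runs a Control Rig toggled by a hooded-variant tag, then applies additive bone rotations and spring-based AnimDynamics.
# EventGraph computes noise rotators, gravity override and skydive alpha.
# It also drives WPO cloth material parameters from spine velocity, switching wind strength and velocity limit when skydiving.
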AnimBlueprint /BRCosmetics/Characters/Player/Male/Medium/Bodies/M_MED_SwissKale/Meshes/Parts/M_MED_SwissKale_FaceAcc_AnimBP
target_skeleton=&/M_MED_SwissKale_FaceAcc_Skeleton
graphs [
 AnimGraph {nodes [
   n1 = Layered blend per bone(
    BasePose=Copy Pose From Mesh() {bCopyCurves=true, bUseAttachedParent=true}
    BlendPoses_0=Copy Pose From Mesh(SourceMeshComponent=Get Body_Mesh())
   ) {layerSetup [{branchFilters [{boneName=upperarm_l}, {boneName=upperarm_twist_01_l}, {boneName=upperarm_twist_02_l}, {boneName=upperarm_twist_01_r}]}]}
   n2 = Control Rig!(Source=@n1, Alpha, hoodActive:bool=Get hoodedVersionEngaged() -> (hoodedVersionEngaged:bool)) -> (Pose) {
    alphaBoolBlend {blendOption=Linear}
    controlRigAssetReference {blueprintRigClass=&/M_MED_SwissKale_FaceAcc_CtrlRig.M_MED_SwissKale_FaceAcc_CtrlRig_C}
   }
   n3 = "Transform (Modify) Bone - Bone: dyn_topDrape_A"!(ComponentPose=Local To Component(LocalPose=@n2), Alpha=1, Rotation="0,0,0") -> (Pose) {rotationMode=BMM_Additive, alphaBoolBlend {blendOption=Linear}, alphaScaleBiasClamp {scale=0.3}, boneToModify {boneName=dyn_topDrape_A}}
   n4 = "Transform (Modify) Bone - Bone: dyn_topDrape_C_1"!(
    ComponentPose="Transform (Modify) Bone - Bone: dyn_topDrape_B"!(ComponentPose=@n3, Alpha=1, Rotation="0,0,0") -> (Pose) {rotationMode=BMM_Additive, alphaBoolBlend {blendOption=Linear}, boneToModify {boneName=dyn_topDrape_B}}
    Alpha=1
    Rotation="0,0,0"
   ) -> (Pose) {rotationMode=BMM_Additive, rotationSpace=BCS_BoneSpace, alphaBoolBlend {blendOption=Linear}, boneToModify {boneName=dyn_topDrape_C_1}}
   n5 = "Transform (Modify) Bone - Bone: dyn_topDrape_C_2"!(ComponentPose=@n4, Alpha=1, Rotation="0,0,0") -> (Pose) {rotationMode=BMM_Additive, rotationSpace=BCS_BoneSpace, alphaBoolBlend {blendOption=Linear}, boneToModify {boneName=dyn_topDrape_C_2}}
   n6 = "Transform (Modify) Bone - Bone: dyn_topDrape_C_3"!(ComponentPose=@n5, Alpha=1, Rotation="0,0,0") -> (Pose) {rotationMode=BMM_Additive, rotationSpace=BCS_BoneSpace, alphaBoolBlend {blendOption=Linear}, boneToModify {boneName=dyn_topDrape_C_3}}
   n7 = Anim Dynamics - Bone: dyn_topDrape_A!(
    ComponentPose="Transform (Modify) Bone - Bone: dyn_bottomDrape_B_2"!(
     ComponentPose="Transform (Modify) Bone - Bone: dyn_bottomDrape_B_1"!(ComponentPose=@n6, Alpha=1, Rotation="0,0,0") -> (Pose) {rotationMode=BMM_Additive, alphaBoolBlend {blendOption=Linear}, boneToModify {boneName=dyn_bottomDrape_B_1}}
     Alpha=1
     Rotation="0,0,0"
    ) -> (Pose) {rotationMode=BMM_Additive, alphaBoolBlend {blendOption=Linear}, boneToModify {boneName=dyn_bottomDrape_B_2}}
    Alpha=1
    GravityOverride:Vector="0,0,-980"
   ) -> (Pose) {
    angularDampingOverride=1
    angularSpringConstant=3000
    bAngularSpring=true
    bOverrideAngularDamping=true
    bUseGravityOverride=true
    gravityScale=0.35
    simulationSpace=RootRelative
    alphaBoolBlend {blendOption=Linear}
    boundBone {boneName=dyn_topDrape_A}
    componentAppliedLinearAccClamp=<3000 3000 3000>
    componentLinearAccScale=<1 1 1>
    componentLinearVelScale=<1 1 1>
    gravityOverride {z=-980}
    physicsBodyDefinitions [{
      boundBone {boneName=dyn_topDrape_A}
      boxExtents=<30 30 30>
      constraintSetup {angularLimitsMax=<5 0 15>, angularLimitsMin=<-5 -20 -15>, angularTarget=<1 0 0>}
      localJointOffset=<3 0 0>
     }]
   }
   n8 = Anim Dynamics - Bone: dyn_topDrape_B!(ComponentPose=@n7, Alpha=1, GravityOverride:Vector="0,0,-980") -> (Pose) {
    angularDampingOverride=1
    angularSpringConstant=3000
    bAngularSpring=true
    bOverrideAngularDamping=true
    bUseGravityOverride=true
    gravityScale=0.35
    simulationSpace=RootRelative
    alphaBoolBlend {blendOption=Linear}
    boundBone {boneName=dyn_topDrape_B}
    componentAppliedLinearAccClamp=<3000 3000 3000>
    componentLinearAccScale=<1 1 1>
    componentLinearVelScale=<1 1 1>
    gravityOverride {z=-980}
    physicsBodyDefinitions [{
      boundBone {boneName=dyn_topDrape_B}
      boxExtents=<30 30 30>
      constraintSetup {angularLimitsMax=<5 0 15>, angularLimitsMin=<-5 -20 -15>, angularTarget=<1 0 0>}
      localJointOffset=<3 0 0>
     }]
   }
   n9 = Anim Dynamics - Bone: dyn_topDrape_C_1!(ComponentPose=@n8, Alpha=1, GravityOverride:Vector="0,0,-980") -> (Pose) {
    angularDampingOverride=1
    angularSpringConstant=3000
    bAngularSpring=true
    bOverrideAngularDamping=true
    bUseGravityOverride=true
    gravityScale=0.35
    simulationSpace=RootRelative
    alphaBoolBlend {blendOption=Linear}
    boundBone {boneName=dyn_topDrape_C_1}
    componentAppliedLinearAccClamp=<3000 3000 3000>
    componentLinearAccScale=<1 1 1>
    componentLinearVelScale=<1 1 1>
    gravityOverride {z=-980}
    physicsBodyDefinitions [{
      boundBone {boneName=dyn_topDrape_C_1}
      boxExtents=<30 30 30>
      constraintSetup {angularLimitsMax=<0 0 0>, angularLimitsMin=<0 -10 0>, angularTarget=<1 0 0>}
      localJointOffset=<3 0 0>
     }]
   }
   n10 = Anim Dynamics - Bone: dyn_topDrape_C_2!(ComponentPose=@n9, Alpha=1, GravityOverride:Vector="0,0,-980") -> (Pose) {
    angularDampingOverride=1
    angularSpringConstant=3000
    bAngularSpring=true
    bOverrideAngularDamping=true
    bUseGravityOverride=true
    gravityScale=0.35
    simulationSpace=RootRelative
    alphaBoolBlend {blendOption=Linear}
    boundBone {boneName=dyn_topDrape_C_2}
    componentAppliedLinearAccClamp=<3000 3000 3000>
    componentLinearAccScale=<1 1 1>
    componentLinearVelScale=<1 1 1>
    gravityOverride {z=-980}
    physicsBodyDefinitions [{
      boundBone {boneName=dyn_topDrape_C_2}
      boxExtents=<30 30 30>
      constraintSetup {angularLimitsMax=<0 0 0>, angularLimitsMin=<0 -20 0>, angularTarget=<1 0 0>}
      localJointOffset=<3 0 0>
     }]
   }
   n11 = Anim Dynamics - Bone: dyn_topDrape_C_3!(ComponentPose=@n10, Alpha=1, GravityOverride:Vector="0,0,-980") -> (Pose) {
    angularDampingOverride=1
    angularSpringConstant=3000
    bAngularSpring=true
    bOverrideAngularDamping=true
    bUseGravityOverride=true
    gravityScale=0.35
    simulationSpace=RootRelative
    alphaBoolBlend {blendOption=Linear}
    boundBone {boneName=dyn_topDrape_C_3}
    componentAppliedLinearAccClamp=<3000 3000 3000>
    componentLinearAccScale=<1 1 1>
    componentLinearVelScale=<1 1 1>
    gravityOverride {z=-980}
    physicsBodyDefinitions [{
      boundBone {boneName=dyn_topDrape_C_3}
      boxExtents=<30 30 30>
      constraintSetup {angularLimitsMax=<15 0 15>, angularLimitsMin=<-15 -20 -15>, angularTarget=<1 0 0>}
      localJointOffset=<3 0 0>
     }]
   }
   n12 = Anim Dynamics - Bone: dyn_bottomDrape_B_1!(ComponentPose=@n11, Alpha=1, GravityOverride:Vector="0,0,-980") -> (Pose) {
    angularDampingOverride=0.9999
    angularSpringConstant=300
    bAngularSpring=true
    bOverrideAngularDamping=true
    bUseGravityOverride=true
    gravityScale=0.35
    simulationSpace=RootRelative
    alphaBoolBlend {blendOption=Linear}
    boundBone {boneName=dyn_bottomDrape_B_1}
    componentAppliedLinearAccClamp=<3000 3000 3000>
    componentLinearAccScale=<1 1 1>
    componentLinearVelScale=<1 1 1>
    gravityOverride {z=-980}
    physicsBodyDefinitions [{
      boundBone {boneName=dyn_bottomDrape_B_1}
      boxExtents=<23 23 23>
      constraintSetup {angularLimitsMax=<0 0 5>, angularLimitsMin=<0 -20 -5>, angularTarget=<1 0 0>}
      localJointOffset=<2 0 0>
     }]
   }
   n13 = Anim Dynamics - Bone: dyn_bottomDrape_B_2!(ComponentPose=@n12, Alpha=1, GravityOverride:Vector="0,0,-980") -> (Pose) {
    angularDampingOverride=0.9999
    angularSpringConstant=300
    bAngularSpring=true
    bOverrideAngularDamping=true
    bUseGravityOverride=true
    gravityScale=0.35
    simulationSpace=RootRelative
    alphaBoolBlend {blendOption=Linear}
    boundBone {boneName=dyn_bottomDrape_B_2}
    componentAppliedLinearAccClamp=<3000 3000 3000>
    componentLinearAccScale=<1 1 1>
    componentLinearVelScale=<1 1 1>
    gravityOverride {z=-980}
    physicsBodyDefinitions [{
      boundBone {boneName=dyn_bottomDrape_B_2}
      boxExtents=<23 23 23>
      constraintSetup {angularLimitsMax=<0 0 15>, angularLimitsMin=<0 -20 -15>, angularTarget=<1 0 0>}
      localJointOffset=<2 0 0>
     }]
   }
   n14 = Anim Dynamics - Bone: dyn_L_hood!(ComponentPose=@n13, Alpha=1, GravityOverride:Vector="0,0,-980") -> (Pose) {
    angularDampingOverride=1
    angularSpringConstant=10000
    bAngularSpring=true
    bOverrideAngularDamping=true
    bUseGravityOverride=true
    gravityScale=0.35
    simulationSpace=RootRelative
    alphaBoolBlend {blendOption=Linear}
    boundBone {boneName=dyn_L_hood}
    componentAppliedLinearAccClamp=<3000 3000 3000>
    componentLinearAccScale=<1 1 1>
    componentLinearVelScale=<1 1 1>
    gravityOverride {z=-980}
    physicsBodyDefinitions [{
      boundBone {boneName=dyn_L_hood}
      boxExtents=<27 27 27>
      constraintSetup {angularLimitsMax=<15 0 15>, angularLimitsMin=<-15 0 -30>, angularTarget=<1 0 0>}
      localJointOffset=<4 2 0>
     }]
   }
   n15 = Anim Dynamics - Bone: dyn_R_hood!(ComponentPose=@n14, Alpha=1, GravityOverride:Vector="0,0,-980") -> (Pose) {
    angularDampingOverride=1
    angularSpringConstant=10000
    bAngularSpring=true
    bOverrideAngularDamping=true
    bUseGravityOverride=true
    gravityScale=0.35
    simulationSpace=RootRelative
    alphaBoolBlend {blendOption=Linear}
    boundBone {boneName=dyn_R_hood}
    componentAppliedLinearAccClamp=<3000 3000 3000>
    componentLinearAccScale=<1 1 1>
    componentLinearVelScale=<1 1 1>
    gravityOverride {z=-980}
    physicsBodyDefinitions [{
      boundBone {boneName=dyn_R_hood}
      boxExtents=<27 27 27>
      constraintSetup {angularLimitsMax=<15 0 15>, angularLimitsMin=<-15 0 -30>, angularTarget=<1 0 0>}
      localJointOffset=<-4 -2 0>
     }]
   }
   n16 = Anim Dynamics - Bone: dyn_hood!(ComponentPose=@n15, Alpha=1, GravityOverride:Vector="0,0,-980") -> (Pose) {
    angularSpringConstant=6000
    bAngularSpring=true
    bOverrideLinearDamping=true
    bUseGravityOverride=true
    gravityScale=0.35
    linearDampingOverride=1
    simulationSpace=RootRelative
    alphaBoolBlend {blendOption=Linear}
    boundBone {boneName=dyn_hood}
    componentAppliedLinearAccClamp=<3000 3000 3000>
    componentLinearAccScale=<1 1 1>
    componentLinearVelScale=<1 1 1>
    gravityOverride {z=-980}
    physicsBodyDefinitions [{
      boundBone {boneName=dyn_hood}
      boxExtents=<40 40 40>
      constraintSetup {angularLimitsMax=<5 0 15>, angularLimitsMin=<-5 0 0>, angularTarget=<1 0 0>, linearAxesMin=<0 -1 0>}
      localJointOffset=<5 0 0>
     }]
   }
   Output Pose(Result=Component To Local(ComponentPose=@n16)) {layerGroup=None}
  ]}
 EventGraph {nodes [
   n1 = Event BlueprintUpdateAnimation()
   > n2 = Sequence()
   n3 = GetGameTimeInSeconds()
   n4 = float * float(A=@n3, B=1.5) {is=title K2Node_PromotableOperator}
   n5 = RandomRotatorFunction!(
    execute=@n2.then_1
    GameTimeInSeconds:double=@n4
    PerlinScale:double=1
    Frequency:Rotator="5,5,5"
    Amplitude:Rotator="5,7,0"
    Offset:Rotator="15,-15,0"
    TimeOffset:double=0
   ) -> (then, RandomRotator:Rotator)
   > Set clothRotator 1(clothRotator 1:Rotator=@n5) -> (Output_Get:Rotator)
   > n6 = RandomRotatorFunction(GameTimeInSeconds=@n4, PerlinScale=1, Frequency="4.5,0,0", Amplitude="10,0,0", Offset="10,0,0", TimeOffset=4)
   > Set clothRotator 2 first(clothRotator 2 first:Rotator=@n6) -> (Output_Get:Rotator)
   > n7 = RandomRotatorFunction(GameTimeInSeconds=@n4, PerlinScale=1, Frequency="4.5,0,0", Amplitude="15,0,0", Offset="5,0,0", TimeOffset=2)
   > Set clothRotator 2 second(clothRotator 2 second:Rotator=@n7) -> (Output_Get:Rotator)
   > n8 = RandomRotatorFunction(GameTimeInSeconds=@n4, PerlinScale=1, Frequency="4.5,0,0", Amplitude="15,0,0", Offset="0, 0, 0", TimeOffset=0)
   > Set clothRotator 2 third(clothRotator 2 third:Rotator=@n8) -> (Output_Get:Rotator)
   n9 = float * float(A=@n3, B=1.3)
   > n10 = RandomRotatorFunction(GameTimeInSeconds=@n9, PerlinScale=1, Frequency="3,0,0", Amplitude="10,0,0", Offset="10,0,0", TimeOffset=2)
   > Set clothRotator 3 first(clothRotator 3 first:Rotator=@n10) -> (Output_Get:Rotator)
   > n11 = RandomRotatorFunction(GameTimeInSeconds=@n9, PerlinScale=1, Frequency="3,0,0", Amplitude="12,0,0", Offset="10,0,0", TimeOffset=0)
   > Set clothRotator 3 second(clothRotator 3 second:Rotator=@n11) -> (Output_Get:Rotator)
   > n12 = Gravity Override State Function DEPRECATED(CharacterPartAnimInstance=Self-Reference(), GameTimeInSeconds=0, GravityOverrideParams=Get Gravity Override Params())
   > Set Gravity Override(Gravity Override=@n12)
   > n13 = Branch(Condition=Get bIsSkyDiving())
   > Set skydiveAlpha(skydiveAlpha:double=0.5) -> (Output_Get:double)
   Set skydiveAlpha(execute=@n13.else, skydiveAlpha=1)
   Is Valid(exec=@n2.then_0, InputObject=Get Skeletal Mesh() -> (Skeletal Mesh:Skeletal Mesh Component Object Reference))
   n14 = Get Material() -> (Material:Material Instance Dynamic Object Reference)
   n15 = Get Overlay Material() -> (Overlay Material:Material Instance Dynamic Object Reference)
   n16 = Get Skeletal Mesh()
   > Set WPO Globals Function(
    Material=@n14
    Overlay Material=@n15
    Skeletal Mesh=@n16
    "Local Noise Transform  Location"="Get Local Noise Transform  Location"() -> ("Local Noise Transform  Location":Vector)
    Local Noise Offset=Get Local Noise Offset()
   )
   > Set WPO Regional Function!(
    execute
    Material:Material Instance Dynamic Object Reference=@n14
    Overlay Material:Material Instance Dynamic Object Reference=@n15
    Skeletal Mesh:Skeletal Mesh Component Object Reference=@n16
    Velocity Transform:Name=spine_05
    Velocity Interp Speed:double=Get Velocity Interp Speed()
    Collision Transform:Name=None
    Collision Transform Axis:EAxis Enum
    Invert Collision Transform Axis:bool=false
    Delta Time X:double=@n1.DeltaTimeX
    Previous Velocity Transform Location:"Vector (by ref)"=Get Previous Velocity Transform Location() -> (Previous Velocity Transform Location:Vector)
    Previous Velocity Transform Velocity:"Vector (by ref)"=Get Previous Velocity Transform Velocity() -> (Previous Velocity Transform Velocity:Vector)
    Region:int=1
   ) -> (then)
   > n17 = Sequence()
   > n18 = Branch(Condition=Get bIsSkyDiving())
   > n19 = Set WPO Pho Wind Intensity(WPO Pho Wind Intensity=1)
   SetScalarParameterValue(
    execute=[@n19.then, @n20.then]
    self=[Get Material(), Get Overlay Material()]
    ParameterName=ClothSim_WindStrength
    Value=Get WPO Pho Wind Intensity()
   )
   n20 = Set WPO Pho Wind Intensity(execute=@n18.else, WPO Pho Wind Intensity=0)
   n21 = Branch(execute=@n17.then_1, Condition=Get bIsSkyDiving())
   > n22 = Set ClothSim_VelocityMax(ClothSim_VelocityMax:double=750) -> (Output_Get:double)
   SetScalarParameterValue(
    execute=[@n23.then, @n22.then]
    self=Get Material()
    ParameterName=ClothSim_VelocityMax
    Value=Get ClothSim_VelocityMax() -> (ClothSim_VelocityMax:double)
   )
   n23 = Set ClothSim_VelocityMax(execute=@n21.else, ClothSim_VelocityMax=5000)
   Event OnCharacterCustomizationLoaded()
   > n24 = Sequence!(execute) -> (then_0, then_1, then_2:Exec)
   n25 = Cast To PlayerPawn_Athena(execute=@n24.then_2, Object=TryGetPawnOwner())
   > Set Body_Mesh(Body_Mesh=GetSkeletalMeshForPartType(self=@n25))
   n26 = Set Skeletal Mesh(execute=@n24.then_1, Skeletal Mesh:Skeletal Mesh Component Object Reference=GetPartSkeletalMeshComponent()) -> (Output_Get:Skeletal Mesh Component Object Reference)
   > n27 = Get Dynamic Material Instance(SkeletalMeshComponent=@n26)
   > Set Material(Material:Material Instance Dynamic Object Reference=@n27) -> (Output_Get:Material Instance Dynamic Object Reference)
   > n28 = GetOrCreateDynamicOverlayMaterial(MeshComponent=@n26, MaterialName=None)
   > Set Overlay Material(Overlay Material:Material Instance Dynamic Object Reference=@n28) -> (Output_Get:Material Instance Dynamic Object Reference)
   n29 = Branch(execute=@n24.then_0, Condition=CharacterItemDefinitionHasAnyMetaTag(Tags=Make Array("[0]"=(TagName="Cosmetics.Variant.Property.Mat1"))))
   > Set hoodedVersionEngaged(hoodedVersionEngaged:bool=true) -> (Output_Get:bool)
   Set hoodedVersionEngaged(execute=@n29.else, hoodedVersionEngaged=false)
  ]}
]
variables [
 Body_Mesh {type=SkeletalMeshComponent}
 clothRotator 1 {type=Rotator}
 clothRotator 2 first {type=Rotator}
 clothRotator 2 second {type=Rotator}
 clothRotator 2 third {type=Rotator}
 clothRotator 3 first {type=Rotator}
 clothRotator 3 second {type=Rotator}
 ClothSim_VelocityMax {type=number}
 Gravity Override {type=Vector}
 Gravity Override Params {type=GravityOverrideParamsStruct, default {skydive_FloatingDown {gravityOverride=<0 0 -500>}}}
 hoodedVersionEngaged {type=boolean}
 Local Noise Offset {type=Vector}
 "Local Noise Transform  Location" {type=Vector}
 Material {type=MaterialInstanceDynamic}
 Overlay Material {type=MaterialInstanceDynamic}
 Previous Velocity Transform Location {type=Vector}
 Previous Velocity Transform Velocity {type=Vector}
 Skeletal Mesh {type=SkeletalMeshComponent}
 skydiveAlpha {type=number}
 Velocity Interp Speed {type=number}
 WPO Pho Wind Intensity {type=number}
]
